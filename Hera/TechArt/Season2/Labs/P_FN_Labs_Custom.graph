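# Labs environment material built on the 2022 four-layer GigaMix base, replacing its emissive with a custom animated glow.
# A tinted emissive, gated by a UV1 layer mask and scanlines, blends with the original, pulses by time and world position, then passes through an RGB rainbow function.
# Shared functions add damage, wind and other WPO, backface normal fixes, Veneer, jungle biome and RVT blending.
# Low-quality mobile flattens specular and roughness and rescales emissive on low quality.
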Material /Hera/TechArt/Season2/Labs/P_FN_Labs_Custom
graph {
 properties {
  bAutomaticallySetUsageInEditor=false
  bTangentSpaceNormal=false
  bUseMaterialAttributes=true
  bUsedWithInstancedStaticMeshes=true
  bUsedWithNanite=true
  ditheredLODTransition=true
  pixelDepthOffsetMode=PDOM_Legacy
  previewMesh=/NarrowFleaJade/Assets/Meshes/NarrowFlea_Gem_DS
  shadingModel=MSM_FromMaterialExpression
  usedShadingModels=Default Lit | Subsurface | Clear Coat | Two Sided Foliage
 }
 nodes [
  n1 = GetMaterialAttributes(None=MaterialFunctionCall() {materialFunction=/Game/Global/Base4Layers_2022/MaterialFunctions/MF_GigaMix_2023}.Result) {attributeGetTypes [#1]}
  n2 = NamedRerouteUsage()
  n3 = Multiply(
   A=Add(A=NamedRerouteUsage()) {constB=0.005}
   B=VectorParameter() {group=Custom, parameterName=Emissive Color Custom, defaultValue=<rgba 1 0 0 1>}.RGB
  )
  n4 = ScalarParameter() {group=Custom, parameterName=Use additive emissive}
  n5 = Add(A=ComponentMask(None=TextureCoordinate() {coordinateIndex=1}) {r=true}) {constB=-0.5}
  n6 = Saturate(None=Round(None=Abs(None=@n5)))
  n7 = Add(A=@n5) {constB=-1}
  n8 = Saturate(None=Round(None=Abs(None=@n7)))
  n9 = Max(
   A=Multiply(A=OneMinus(None=@n6), B=ScalarParameter() {group=Custom, parameterName=L1, sortPriority=-4})
   B=Multiply(A=OneMinus(None=@n8), B=ScalarParameter() {defaultValue=1, group=Custom, parameterName=L2, sortPriority=-3})
  )
  n10 = LinearInterpolate(
   A=Multiply(A=Step(X=@n2) {constX=0.9, constY=0.01}, B=@n3)
   B=@n3
   Alpha=LinearInterpolate(B=Saturate(None=@n4), Alpha=NamedRerouteDeclaration(None=@n9) {name=Mask Layer, nodeColor=<rgba 1 0 0.5859 1>})
  )
  n11 = Add(A=MaterialFunctionCall!(
    ScanLine_Tiling_1=ScalarParameter() {defaultValue=3, group=Custom, parameterName=Scan Tile}
    Scanline_PannerSpeed_1
    ScanLine_Tiling=ScalarParameter() {defaultValue=70, group=Custom, parameterName=Scan Tile 2}
    Scanline_PannerSpeed
    UVs
   ) {materialFunction=/Game/Weapons/WeaponSkins/MaterialFunction/ScanLines}.Result) {constB=0.2}
  n12 = ScalarParameter() {defaultValue=0.5, group=Custom, parameterName=OG Emissive Blend}
  n13 = ObjectPositionWS()
  n14 = Constant3Vector() {constant=<rgba 3.65 2.65 1.4539 0>}
  n15 = Add(A=Time(), B=Multiply(A=DotProduct(A=@n13, B=@n14)) {constB=50})
  n16 = Multiply(
   A=LinearInterpolate(A=NamedRerouteUsage(), B=Multiply(A=@n10, B=@n11), Alpha=Saturate(None=@n12)) {constAlpha=1}
   B=MaterialFunctionCall!(
    Sine Phase=Divide(A=@n15) {constB=6}
    Value 1=ScalarParameter() {defaultValue=0.5, group=Custom, parameterName=Min Emissive}
    Value 2=ScalarParameter() {defaultValue=0.9, group=Custom, parameterName=Max Emissive}
   ) {materialFunction=/Engine/Functions/Engine_MaterialFunctions02/Utility/Sine_Remapped}.Remapped Sine
  )
  n17 = Multiply(A=@n16, B=ScalarParameter() {defaultValue=1, group=Custom, parameterName=Mobile Emissive}) {constB=0.5}
  n18 = MaterialFunctionCall!(
   Coords
   MA=SetMaterialAttributes(MaterialAttributes=@n1.MaterialAttributes, Emissive Color=QualitySwitch(Default=@n16, Low=@n17)) {attributeSetTypes [#1]}
  ) {materialFunction=/Hera/Sets/AxisMandate/TechArt/MF_RGB_Emissive}
  MaterialExpressionReroute_70 = Reroute(None=@n18."Rainbow Emissive (MA)")
  n19 = ScalarParameter() {
   desc=Documentation link
   group="0 (Expand)  Documentation (RMB - Copy display name) and Debug view"
   parameterName=https://confluence.it.epicgames.com/display/FORT/Fortnite+Base+Material+2022
  }
  n20 = SetMaterialAttributes!(MaterialAttributes=@MaterialExpressionReroute_70.MaterialAttributes, World Position Offset=Multiply(A=@n19)) {attributeSetTypes [#2]}
  MaterialExpressionReroute_71 = Reroute(None=@n20)
  MaterialExpressionReroute_69 = Reroute(None=@MaterialExpressionReroute_71.Output)
  n21 = MaterialFunctionCall(Material_In=@MaterialExpressionReroute_69.MaterialAttributes) {materialFunction=&MF_FN_DestructibleDamage}
  n22 = MaterialFunctionCall!(
   In=MaterialFunctionCall!(Material Attributes=MaterialFunctionCall!(In=@n21.Material_Out) {materialFunction=/Game/Athena/WinterArt/Materials/MF_WPO_WindNew}.Result) {materialFunction=/Game/Creative/Sets/MilitaryBase/Props/MF_FanSpin}.Result
   UseSphericalMappingOverride
  ) {materialFunction=/Game/Packages/Fortress_SharedMaterials/Base_Material_Functions/MF_CurieFallbackEffect}
  n23 = MaterialFunctionCall!(In=MaterialFunctionCall!(In=MaterialFunctionCall!(In=@n22.Result, Custom Mask, Custom Rotation Axis) {materialFunction=/Game/Athena/WinterArt/Materials/MF_WindWPO_v2}.Result) {materialFunction=/Game/Global/Base4Layers_2022/MaterialFunctions/MF_SnowAndIceArchitecture}.Result) {materialFunction=/Game/Global/Base4Layers_2022/MaterialFunctions/MF_VibrationRattle}
  MaterialExpressionReroute_68 = Reroute(None=@n23.Result)
  n24 = MaterialFunctionCall!(In=@MaterialExpressionReroute_68.Result) {materialFunction=/Game/Athena/WinterArt/Materials/MF_WPOAnim}
  Composite!(None=@n24.Result) {desc=Transform to World in case we will use RVT after this point, subgraphName=Tangent to World}
  Composite!(MaterialAttributes=@n18."Rainbow Emissive (MA)") {subgraphName=Documentation Link}
  Composite(MaterialAttributes=@MaterialExpressionReroute_71.Output) {desc=Game Shared WPO and Fire Effect, subgraphName=WPO And Shared Functions}
  MaterialExpressionReroute_1 = Reroute(None=@n31.RVT H)
  MaterialExpressionReroute_3 = Reroute(None=@n24.Result)
  n25 = GetMaterialAttributes(None=@MaterialExpressionReroute_3.Input) {attributeGetTypes [#3]}
  n26 = Transform(None=@n25.Normal)
  n27 = LinearInterpolate(A=Multiply(A=@n26) {constB=-1}, B=@n26, Alpha=Saturate(None=TwoSidedSign())) {desc="Fixes normals\r\non backfaces"}
  n28 = SetMaterialAttributes!(
   MaterialAttributes=@n25.MaterialAttributes
   Normal=StaticSwitchParameter(True=MaterialProxyReplace(Realtime=@n27, MaterialProxy=@n25.Normal), False=@n27) {parameterName=Force TangentSpaceNormalsFor HLOD ProxyReplace}
  ) {attributeSetTypes [#3]}
  MaterialExpressionReroute_2 = Reroute(None=@n28)
  n29 = MaterialFunctionCall!(In=MaterialFunctionCall!(In=MaterialFunctionCall!(In=@MaterialExpressionReroute_2.Output) {materialFunction=/Game/ArtTools/Tech_Art/DirectFX/Assets/MaterialFunctions/MF_DirectFX}.Result) {materialFunction=/Game/Global/MaterialFunctions/MF_GlobalEmissiveMultiplier_MPC}.Result) {materialFunction=/Game/ArtTools/Tech_Art/Veneer/MaterialFunctions/MF_Veneer}
  MaterialExpressionReroute_8 = Reroute(None=@n29.Result)
  n30 = MaterialFunctionCall!(AO, RVT_Height=@MaterialExpressionReroute_1."RVT_Height (S)", For Buildings=@MaterialExpressionReroute_8."For Moss (MA)", For Moss) {materialFunction=/Game/Environments/Asteria/Rocks/Materials/MaterialFunctions/MF_JungleBiome}
  MaterialExpressionReroute_0 = Reroute(None=@n30.Buildings Jungl)
  n31 = MaterialFunctionCall!(
   In=@MaterialExpressionReroute_0.Result
   "RVT Height Map (Roughness)"=GetMaterialAttributes(None=@n29.Result) {attributeGetTypes [#4]}.Roughness
  ) {materialFunction=/Game/Global/Base4Layers_2022/MaterialFunctions/MF_RVTBlending}
  Composite!(InputPin=@n31.Result) {subgraphName=Mobile Opt}
  Composite!("For Moss (MA)"=@n29.Result, "RVT_Height (S)"=@n31.RVT H) {subgraphName="Seasonal "}
  Constant2Vector()
  MaterialExpressionReroute_7 = Reroute(None=@n31.Result)
  n32 = GetMaterialAttributes(None=@MaterialExpressionReroute_7.InputPin)
  n33 = Constant() {r=0.5}
  n34 = Constant() {r=0.8}
  n35 = ShadingPathSwitch(
   Default=@MaterialExpressionReroute_7.InputPin
   Mobile=QualitySwitch(
    Default=@MaterialExpressionReroute_7.InputPin
    Low=SetMaterialAttributes!(MaterialAttributes=@n32.MaterialAttributes, Specular=@n33, Roughness=@n34) {attributeSetTypes [#5, #4]}
    High=@MaterialExpressionReroute_7.InputPin
    Medium=@MaterialExpressionReroute_7.InputPin
    Epic=@MaterialExpressionReroute_7.InputPin
   )
  )
  MaterialExpressionReroute_6 = Reroute(None=@n35)
  MaterialFunctionCall(In=MaterialFunctionCall(MA In=@MaterialExpressionReroute_6.Output) {materialFunction=&MF_CreativeMoveTool}.MA Out) {materialFunction=&MF_GlobalEmissiveMultiplier_MPC}
  NamedRerouteDeclaration(None=@n1.Emissive Color) {name=OG Emissive, nodeColor=<rgba 0 0.5781 1 1>}
  PinBase() {reroutePins [Input {expression=&self:MaterialExpressionReroute_3}]}
  PinBase!(Output=@n28) {reroutePins [Output {expression=&self:MaterialExpressionReroute_2}]}
  PinBase!() {reroutePins [
    "For Moss (MA)" {expression=&self:MaterialExpressionReroute_8}
    "RVT_Height (S)" {expression=&self:MaterialExpressionReroute_1}
   ]}
  PinBase!(Result=@n30.Buildings Jungl) {reroutePins [Result {expression=&self:MaterialExpressionReroute_0}]}
  PinBase!() {reroutePins [MaterialAttributes {expression=&self:MaterialExpressionReroute_70}]}
  PinBase!(Output=@n20) {reroutePins [Output {expression=&self:MaterialExpressionReroute_71}]}
  PinBase!() {reroutePins [MaterialAttributes {expression=&self:MaterialExpressionReroute_69}]}
  PinBase!(Result=@n23.Result) {reroutePins [Result {expression=&self:MaterialExpressionReroute_68}]}
  PinBase!() {reroutePins [InputPin {expression=&self:MaterialExpressionReroute_7}]}
  PinBase!(Output=@n35) {reroutePins [Output {expression=&self:MaterialExpressionReroute_6}]}
 ]
}
